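# Sound cue that plays the equip foley for the Figment scoped pistol.
# A fixed short delay feeds a single wave player, which reuses the scoped pistol reload-end sample.
# It routes to the weapon foley sound class with the weapon reload attenuation.
# Its concurrency override prevents new instances from starting once the limit is reached.
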SoundCue /Figment_Pistol_Scoped_Athena/Sounds/SC_ScopedPistol_Equip
root=Delay_0
nodes [
 {id=Delay_0, children [WavePlayer_1], properties {delayMax=0.2, delayMin=0.2}}
 {id=WavePlayer_1, properties {soundWaveAssetPtr=/Game/Sounds/Fort_Weapons/pistol/Scoped/ScopedPistol_Reload_End_02}}
]
properties {
 attenuationSettings=&Fort_Weapon_Reload_Att
 bOverrideConcurrency=true
 maxDistance=2250
 soundClassObject=&Fort_WeaponFoley
 volumeMultiplier=1.153
 concurrencyOverrides {resolutionRule=PreventNew}
}
