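# Scoped pistol gunfire cue for the Figment weapon, layering close, medium and distant shot sounds.
# Each layer runs through a modulator, its own distance attenuation and a sound class override (third-person fire or distant fire) before nested mixers blend them.
# A quality switch gives the first level three random variants per layer, while the other three levels share a branch with one wave per layer.
# Local and global gunfire concurrency sets limit playback.
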SoundCue /Figment_Pistol_Scoped_Athena/Sounds/ScopedPistol_Fire_Figment_Cue
root=QualityLevel_0
nodes [
 {id=Attenuation_0, children [Modulator_2], properties {attenuationSettings=&Fort_Weapon_Projectile_Close_Athena_Att}}
 {
  id=Attenuation_1
  children [Modulator_0]
  properties {attenuationSettings=&Fort_Weapon_Projectile_Distant_Mono_Athena_Att}
 }
 {
  id=Attenuation_6
  children [Modulator_1]
  properties {attenuationSettings=&Fort_Weapon_Projectile_Medium_Mono_Athena_Att}
 }
 {id=Attenuation_7, children [Modulator_13], properties {attenuationSettings=&Fort_Weapon_Projectile_Close_Athena_Att}}
 {
  id=Attenuation_8
  children [Modulator_11]
  properties {attenuationSettings=&Fort_Weapon_Projectile_Distant_Mono_Athena_Att}
 }
 {
  id=Attenuation_9
  children [Modulator_12]
  properties {attenuationSettings=&Fort_Weapon_Projectile_Medium_Mono_Athena_Att}
 }
 {id=Mixer_0, children [Mixer_1,SoundClass_1], properties {inputVolume [1,0.45]}}
 {id=Mixer_1, children [SoundClass_0,SoundClass_6], properties {inputVolume [0.9,0.7]}}
 {id=Mixer_10, children [SoundClass_7,SoundClass_9], properties {inputVolume [0.9,0.7]}}
 {id=Mixer_9, children [Mixer_10,SoundClass_8], properties {inputVolume [1,0.45]}}
 {id=Modulator_0, children [Random_1]}
 {id=Modulator_1, children [Random_0], properties {volumeMax=2.1, volumeMin=2}}
 {id=Modulator_11, children [WavePlayer_14]}
 {id=Modulator_12, children [WavePlayer_12], properties {volumeMax=1.1, volumeMin=1}}
 {id=Modulator_13, children [WavePlayer_11], properties {pitchMax=1.1, pitchMin=1, volumeMax=1.1, volumeMin=1}}
 {id=Modulator_2, children [Random_2], properties {pitchMax=1.1, pitchMin=1, volumeMax=1.1, volumeMin=1}}
 {id=QualityLevel_0, children [Mixer_0,Mixer_9,Mixer_9,Mixer_9]}
 {id=Random_0, children [WavePlayer_2, WavePlayer_3, WavePlayer_13], properties {weights [1,1,1]}}
 {id=Random_1, children [WavePlayer_15, WavePlayer_16, WavePlayer_17], properties {weights [1,1,1]}}
 {id=Random_2, children [WavePlayer_0,WavePlayer_1,WavePlayer_4], properties {weights [1,1,1]}}
 {
  id=SoundClass_0
  children [Attenuation_0]
  properties {soundClassOverride=&Fort_WeaponRanged_Fire_3P_PSM}
 }
 {
  id=SoundClass_1
  children [Attenuation_1]
  properties {soundClassOverride=&Fort_WeaponRanged_Fire_Distant}
 }
 {
  id=SoundClass_6
  children [Attenuation_6]
  properties {soundClassOverride=&Fort_WeaponRanged_Fire_3P_PSM}
 }
 {
  id=SoundClass_7
  children [Attenuation_7]
  properties {soundClassOverride=&Fort_WeaponRanged_Fire_3P_PSM}
 }
 {
  id=SoundClass_8
  children [Attenuation_8]
  properties {soundClassOverride=&Fort_WeaponRanged_Fire_Distant}
 }
 {
  id=SoundClass_9
  children [Attenuation_9]
  properties {soundClassOverride=&Fort_WeaponRanged_Fire_3P_PSM}
 }
 {id=WavePlayer_0, properties {soundWaveAssetPtr=/Game/Sounds/Fort_Weapons/pistol/Scoped/ScopedPistol_Fire_Close_01}}
 {id=WavePlayer_1, properties {soundWaveAssetPtr=/Game/Sounds/Fort_Weapons/pistol/Scoped/ScopedPistol_Fire_Close_02}}
 {id=WavePlayer_11, properties {soundWaveAssetPtr=&ScopedPistol_Fire_Close_01}}
 {id=WavePlayer_12, properties {soundWaveAssetPtr=/Figment_Pistol_Scoped_Athena/Sounds/Wavs/ScopedPistol_Fire_Medium_01}}
 {id=WavePlayer_13, properties {soundWaveAssetPtr=/Figment_Pistol_Scoped_Athena/Sounds/Wavs/ScopedPistol_Fire_Medium_03}}
 {id=WavePlayer_14, properties {soundWaveAssetPtr=/Figment_Pistol_Scoped_Athena/Sounds/Wavs/ScopedPistol_Fire_Distant_01}}
 {id=WavePlayer_15, properties {soundWaveAssetPtr=&ScopedPistol_Fire_Distant_01}}
 {id=WavePlayer_16, properties {soundWaveAssetPtr=/Figment_Pistol_Scoped_Athena/Sounds/Wavs/ScopedPistol_Fire_Distant_02}}
 {id=WavePlayer_17, properties {soundWaveAssetPtr=/Figment_Pistol_Scoped_Athena/Sounds/Wavs/ScopedPistol_Fire_Distant_03}}
 {id=WavePlayer_2, properties {soundWaveAssetPtr=&ScopedPistol_Fire_Medium_01}}
 {id=WavePlayer_3, properties {soundWaveAssetPtr=/Figment_Pistol_Scoped_Athena/Sounds/Wavs/ScopedPistol_Fire_Medium_02}}
 {id=WavePlayer_4, properties {soundWaveAssetPtr=/Game/Sounds/Fort_Weapons/pistol/Scoped/ScopedPistol_Fire_Close_03}}
]
properties {
 maxDistance=27500
 volumeMultiplier=1
 attenuationOverrides {distanceAlgorithm=NaturalSound, falloffDistance=5000, attenuationShapeExtents {x=1000}}
 concurrencySet [&Weapon_Gunfire_Local_Conc, &Weapon_Gunfire_Global_Conc]
}
